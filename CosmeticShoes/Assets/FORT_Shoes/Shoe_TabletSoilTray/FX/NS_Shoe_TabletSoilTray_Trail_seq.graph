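# Niagara trail effect for the TabletSoilTray cosmetic shoe, drawing blue ribbons from the character's skeletal mesh while running.
# An attached ribbon emitter stops spawning when skydiving or hiding foot FX, and narrows and fades ribbons through curves.
# In gameplay, ribbon alpha follows sampled speed, scaled by User.TrailOn; front end gets full alpha and longer lifetime.
# A sprite smoke emitter and two alternate ribbon color passes are present but disabled.
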NiagaraSystem /CosmeticShoes/Assets/FORT_Shoes/Shoe_TabletSoilTray/FX/NS_Shoe_TabletSoilTray_Trail_seq
dependencies {
 usedDataInterfaces [&NiagaraDataInterfaceCurve, &NiagaraDataInterfaceColorCurve, &NiagaraDataInterfaceArrayFloat]
 usedDynamicInputs [&SetBoolByFloatComparison, &Multiply_Vector]
 usedModules [
  &self:SystemSpawnScript.SystemScriptSource.SystemScriptGraph.NiagaraNodeAssignment_0.SetVariables_#1
  &self:SystemSpawnScript.SystemScriptSource.SystemScriptGraph.NiagaraNodeAssignment_3.SetVariables_#2
  &SystemState
  &SampleSkeletalMesh
  &Timeline
  &self:SystemSpawnScript.SystemScriptSource.SystemScriptGraph.NiagaraNodeAssignment_1.SetVariables_#3
  &self:SystemSpawnScript.SystemScriptSource.SystemScriptGraph.NiagaraNodeAssignment_2.SetVariables_#4
  &self:SystemSpawnScript.SystemScriptSource.SystemScriptGraph.NiagaraNodeAssignment_5.SetVariables_#5
 ]
}
emitters {
 FN_AttachedRibbon001 {
  emitterUpdateScript [
   &EmitterState()
   &FN_SpawnRibbon_v2(
    Number of Ribbons:NiagaraInt32=&MakeCustomIntFromBool(Bool=$User.IsSkydiving, True INT=0, False INT=2)
    Spawn Mode=::NewEnumerator0
    SpawnRate:NiagaraFloat=&MakeCustomFloatFromBool(
     Bool=$User.IsSkydiving
     True Float=0
     False Float=&MakeCustomFloatFromBool(
      Bool=$User.HideFootFX
      True Float=0
      False Float=&MakeCustomFloatFromBool(Bool=$System.isRunning, True Float=12, False Float=0)
     )
    )
   )
   &FN_Integrator() {enabled=false}
  ]
  particleSpawnScript [
   /Game/Effects/Niagara/Modules/Ribbons/FN_InitializeRibbon_v4(
    Enable Skeletal Position=true
    Lifetime:NiagaraFloat=&FN_FrontEndFloatModifier(Gameplay=0.17, Front End=0.45)
    Ribbon Width=7
    SkeletalMesh:NiagaraDataInterfaceSkeletalMesh=$System.OnlyRibbons
   )
   &FN_FlipRibbonV()
   &AddVelocity(Velocity=<40 -200 25>)
   &DynamicMaterialParameters(
    Index 0 Param 1=1
    Index 0 Param 2=0.75
    Index 0 Param 3=1
    Index 0 Param 4=1
    Module.Index0Param0Write=true
    Module.Index0Param1Write=true
    Module.Index0Param2Write=true
    Module.Index0Param3Write=true
   )
   /SetVariables(Particles.RibbonWidth=&FloatFromCurve(
     FloatCurve=DataInterface() {propertyValues {Curve {keys [
         {arriveTangent=-0.3053, interpMode=RCIM_Cubic, leaveTangent=-0.3053, tangentMode=RCTM_User, time=0, value=1}
         {arriveTangent=-0.5499, interpMode=RCIM_Cubic, leaveTangent=-0.5499, tangentMode=RCTM_User, time=1, value=0.5}
        ]}}}
     Scale Curve=20
    )) {isSetParameters=true}
   /SetVariables(Particles.R2_Width:NiagaraFloat=&Multiply_Float(A=$Particles.RibbonWidth, B=1.25)) {enabled=false, isSetParameters=true}
   /SetVariables(Particles.R3_Width:NiagaraFloat=&Multiply_Float(A=$Particles.RibbonWidth, B=0.7)) {enabled=false, isSetParameters=true}
   &AddVelocity(
    Additional Quaternion Rotation=true
    Additional Yaw / Pitch / Roll:Vector3f=&MaskVectorBySpawnGroup(Masked Vector=<-80 0 -10>, Passthrough Vector=<100 0 -10>)
    Rotation Mode=::NewEnumerator3
    Rotation Quaternion:Quat4f=$Particles.SampledRotation
    Velocity=<0 -200 50>
    Velocity Speed Scale=&FN_FrontEndFloatModifier(Gameplay=0, Front End=1)
   ) {enabled=false, module=AddVelocity001}
  ]
  particleUpdateScript [
   &InheritSourceMovement(
    Applied Position Scale=&VectorFromFloat(Value=&FloatFromCurve(FloatCurve=DataInterface() {propertyValues {Curve {keys [
          {arriveTangent=-0.2727, interpMode=RCIM_Cubic, leaveTangent=-0.2727, tangentMode=RCTM_User, time=0, value=1}
          {arriveTangent=-0.5057, interpMode=RCIM_Cubic, leaveTangent=-0.5057, tangentMode=RCTM_User, time=1, value=0.14}
         ]}}}))
    ApplyForce=false
    ApplyPosition=true
    ApplyPositionFalloff=false
   )
   &SolveForcesAndVelocity(Clamp Velocity=false, Limit Acceleration=false)
   /Game/Effects/Niagara/Modules/Ribbons/FN_UpdateRibbon_v4(Enable Skeletal Position=true, SkeletalMesh:NiagaraDataInterfaceSkeletalMesh=$System.OnlyRibbons)
   /SetVariables(Particles.AlphaScale:NiagaraFloat=&FN_FrontEndFloatModifier(
     Gameplay=&RemapRange(
      Input Value To Remap=&VectorLength(Vector=$Particles.FN_UpdateRibbon_v4.SampledVelocity)
      Input Min=400
      Input Max=1000
      Output Min=0
      Output Max=1
     )
     Front End=1
    )) {isSetParameters=true}
   &FN_RibbonMaterial(
    Color=<rgba 0.2535 0.7395 2 1>
    Color Pan U Speed=0.1
    Opacity Pan U Speed=1.5
    Opacity Power=0.5
    Opacity Scale U=2
    Scale Alpha:NiagaraFloat=&FloatFromCurve(
     FloatCurve=DataInterface() {propertyValues {Curve {keys [
         {arriveTangent=-1.0872, interpMode=RCIM_Cubic, leaveTangent=-1.0872, tangentMode=RCTM_User, time=0, value=0.5}
         {arriveTangent=-0.3496, interpMode=RCIM_Cubic, leaveTangent=-0.3496, tangentMode=RCTM_User, time=1, value=0}
        ]}}}
     Scale Curve=&Multiply_Float(A=$Particles.AlphaScale, B=$User.TrailOn)
    )
   )
   &FN_RibbonMaterial(
    Color=<rgba 2.5 0.0791 0.001 1>
    Color Pan U Speed=0.1
    Opacity Pan U Speed=1
    Opacity Scale U=0.25
    Scale Alpha:NiagaraFloat=&FloatFromCurve(
     FloatCurve=DataInterface() {propertyValues {Curve {keys [
         {interpMode=RCIM_Cubic, tangentMode=RCTM_User, time=0.3, value=0.5}
         {arriveTangent=-2.4482, interpMode=RCIM_Cubic, leaveTangent=-2.4482, tangentMode=RCTM_User, time=1, value=0}
        ]}}}
     Scale Curve=$Particles.AlphaScale
    )
    Write to Module Namespace=true
   ) {enabled=false, module=FN_RibbonMaterial001}
   &FN_RibbonMaterial(
    Color=&ColorFromCurve(ColorCurve=DataInterface() {propertyValues {
       AlphaCurve {keys [<at 0 1 RCIM_Cubic>,<at 1 0 RCIM_Cubic>]}
       BlueCurve {keys [<at 0 0 RCIM_Cubic>,<at 0.4328 10 RCIM_Cubic>]}
       GreenCurve {keys [<at 0 6.6349 RCIM_Cubic>,<at 0.4328 0 RCIM_Cubic>]}
       RedCurve {keys [<at 0 20 RCIM_Cubic>,<at 0.4328 0.364 RCIM_Cubic>]}
      }})
    Color Pan U Speed=0.1
    Opacity Power=0.3
    Opacity Scale U=0.25
    Scale Alpha:NiagaraFloat=&FloatFromCurve(
     FloatCurve=DataInterface() {propertyValues {Curve {keys [{interpMode=RCIM_Cubic, tangentMode=RCTM_User, time=0.3, value=0.5}, {interpMode=RCIM_Cubic, tangentMode=RCTM_User, time=1, value=0}]}}}
     Scale Curve=$Particles.AlphaScale
    )
    Write to Module Namespace=true
   ) {enabled=false, module=FN_RibbonMaterial002}
  ]
  rendererClasses [&NiagaraRibbonRendererProperties]
  renderers [
   {rendererClass=&NiagaraRibbonRendererProperties}
   {rendererClass=&NiagaraRibbonRendererProperties, rendererIndex=1}
  ]
 }
 Smoke {
  bEnabled=false
  emitterUpdateScript [
   &EmitterState()
   &SpawnRate(
    Spawn Probability=&MakeCustomFloatFromBool(Bool=$User.IsSkydiving, True Float=0, False Float=1)
    SpawnRate=&MakeCustomFloatFromBool(
     Bool=$User.IsSkydiving
     True Float=0
     False Float=&MakeCustomFloatFromBool(
      Bool=$User.HideFootFX
      True Float=0
      False Float=&MakeCustomFloatFromBool(Bool=$System.isRunning, True Float=100, False Float=0)
     )
    )
    Use Spawn Probability=true
   )
  ]
  particleSpawnScript [
   &InitializeParticle(
    Color=&MakeCustomLinearColorFromBool(
     Bool=&Boolean_OrOperation(Bool1=$User.bIsFrontEnd, Bool2=$User.bIsFrontEndPreview)
     True Color=<rgba 0.1562 0.4987 10 1>
     False Color=<rgba 0.3168 0.4687 2.5 0.13>
    )
    Lifetime=&FN_FrontEndFloatModifier(Gameplay=0.3, Front End=0.85)
    Sprite Rotation Mode=::NewEnumerator2
    Sprite Size Mode=::NewEnumerator3
    Uniform Sprite Size=&UniformRangedFloat(Minimum=37, Maximum=52)
    UsePositionOffset=false
    Write Material Random=true
   ) {module=InitializeParticle001}
   &FN_SetSpawnGroup(Number of Spawn Groups=2)
   &SkeletalMeshLocation(Bone / Socket Index=$Particles.SpawnGroup, Bone Sampling Mode=::NewEnumerator4, MaskSpawnGroup=false, Skeletal Mesh=$System.Skeletal Mesh) {module=SkeletalMeshLocation001}
   /SetVariables(Particles.InitSpriteRotRate=&UniformRangedFloat(Minimum=-1, Maximum=1)) {isSetParameters=true}
   &AddVelocity(
    Additional Quaternion Rotation=true
    Additional Yaw / Pitch / Roll:Vector3f=&MaskVectorBySpawnGroup(Masked Vector=<0 0 180>, Passthrough Vector=<0 0 0>)
    Rotation Mode=::NewEnumerator3
    Rotation Quaternion:Quat4f=$Particles.SkeletalMeshLocation001.SampledRotation
    Velocity=<0 120 30>
    Velocity Speed Scale=&FN_FrontEndFloatModifier(Gameplay=0, Front End=1)
   ) {enabled=false, module=AddVelocity001}
   &ApplyInitialForces()
  ]
  particleUpdateScript [
   &ParticleState()
   &SpriteRotationRate(Rotation Rate=&Multiply_Float(
     A=$Particles.InitSpriteRotRate
     B=&FloatFromCurve(
      FloatCurve=DataInterface() {propertyValues {Curve {keys [
          {arriveTangent=-1.9214, interpMode=RCIM_Cubic, leaveTangent=-1.9214, tangentMode=RCTM_User, time=-0.0046, value=1.0113}
          {arriveTangent=-0.2496, interpMode=RCIM_Cubic, leaveTangent=-0.2496, tangentMode=RCTM_User, time=0.9896, value=0.16}
         ]}}}
      CurveIndex=$Emitter.Age
     )
    ))
   &ScaleColor(
    Scale Alpha=&FloatFromCurve(
     FloatCurve=DataInterface() {propertyValues {Curve {
        defaultValue=0
        keys [
         {interpMode=RCIM_Cubic, tangentMode=RCTM_User, time=0, value=0}
         {interpMode=RCIM_Cubic, tangentMode=RCTM_User, time=0.2, value=0.975}
         {tangentMode=RCTM_User, time=1, value=0}
        ]
       }}}
     Scale Curve=$User.TrailOn
    )
    ScaleA=true
    ScaleRGB=true
   ) {module=ScaleColor001}
   &Drag(Use Linear Drag=true, Use Rotational Drag=false)
   &InheritSourceMovement(
    Applied Position Scale=&MakeVector(
     X=&FloatFromCurve(FloatCurve=DataInterface() {propertyValues {Curve {
         defaultValue=1
         keys [
          {arriveTangent=-0.8337, interpMode=RCIM_Cubic, leaveTangent=-0.8337, tangentMode=RCTM_User, time=0, value=0.7}
          {arriveTangent=-0.2374, interpMode=RCIM_Cubic, leaveTangent=-0.2374, tangentMode=RCTM_User, time=1, value=0.3}
         ]
        }}})
     Y=&FloatFromCurve(FloatCurve=DataInterface() {propertyValues {Curve {
         defaultValue=1
         keys [
          {arriveTangent=-0.5997, interpMode=RCIM_Cubic, leaveTangent=-0.5997, tangentMode=RCTM_User, time=0, value=0.7}
          {arriveTangent=-0.5578, interpMode=RCIM_Cubic, leaveTangent=-0.5578, tangentMode=RCTM_User, time=1, value=0.3}
         ]
        }}})
     Z=0.2
    )
    ApplyForce=false
    ApplyPosition=true
    ApplyPositionFalloff=false
   )
   &SolveForcesAndVelocity(Clamp Velocity=false, Limit Acceleration=false)
   &ScaleSpriteSize(Scale Factor=&Vector2DFromFloat(Value=&FloatFromCurve(FloatCurve=DataInterface() {propertyValues {Curve {
         defaultValue=1.4923
         keys [
          {arriveTangent=0.8474, interpMode=RCIM_Cubic, leaveTangent=0.8474, tangentMode=RCTM_User, time=0, value=0.45}
          {interpMode=RCIM_Cubic, tangentMode=RCTM_User, time=1, value=1}
         ]
        }}})))
   &DynamicMaterialParameters(
    Index 0 Param 1=&FloatFromCurve(FloatCurve=DataInterface() {propertyValues {Curve {keys [<at 0 0 RCIM_Cubic>,<at 1.0011 0.2>]}}})
    Index 0 Param 4=1
    Index 1 Param 1=&FloatFromCurve(
     FloatCurve=DataInterface() {propertyValues {Curve {keys [<at 0 0>,<at 0.25 1>]}}}
     Scale Curve=&RandomRangeFloat(Minimum=2, Maximum=1)
    )
    Index 1 Param 2=&FloatFromCurve(
     FloatCurve=DataInterface() {propertyValues {Curve {keys [
         {interpMode=RCIM_Cubic, leaveTangent=3.7422, tangentMode=RCTM_Break, time=0, value=0.2}
         {interpMode=RCIM_Cubic, leaveTangent=0.9066, time=1, value=2}
        ]}}}
     Scale Curve=&RandomRangeFloat(Minimum=0.5, Maximum=1)
    )
    Index 1 Param 3=&FloatFromCurve(FloatCurve=DataInterface() {propertyValues {Curve {keys [
         {interpMode=RCIM_Cubic, leaveTangent=2.8445, tangentMode=RCTM_Break, time=0, value=0.2}
         {interpMode=RCIM_Cubic, leaveTangent=63.2978, time=0.25, value=1}
        ]}}})
    Index 1 Param 4=&FloatFromCurve(
     FloatCurve=DataInterface() {propertyValues {Curve {keys [<at 0 1>,<at 1 0.8>]}}}
     Scale Curve=&RandomRangeFloat(Minimum=0.5, Maximum=1)
    )
    Index 2 Param 1=&FloatFromCurve(FloatCurve=DataInterface() {propertyValues {Curve {keys [{interpMode=RCIM_Cubic, leaveTangent=3.2309, tangentMode=RCTM_Break, time=0.1, value=0}, <at 1 1>]}}})
    Index 2 Param 2=$Particles.MaterialRandom
    Index 2 Param 3=&FloatFromCurve(FloatCurve=DataInterface() {propertyValues {Curve {keys [
         {interpMode=RCIM_Cubic, tangentMode=RCTM_Break, time=0, value=1}
         {interpMode=RCIM_Cubic, leaveTangent=11.1233, tangentMode=RCTM_Break, time=1, value=2}
        ]}}})
    Index 2 Param 4=0.75
    Module.Param0WriteEnabled=true
    Module.Param0WriteEnabled1=true
    Module.Param0WriteEnabled2=true
    Module.Param1WriteEnabled=false
    Module.Param1WriteEnabled1=true
    Module.Param1WriteEnabled2=true
    Module.Param2WriteEnabled=true
    Module.Param2WriteEnabled1=true
    Module.Param2WriteEnabled2=true
    Module.Param3WriteEnabled=true
    Module.Param3WriteEnabled1=true
    Module.Param3WriteEnabled2=true
    Write Parameter Index 1=true
    Write Parameter Index 2=true
   )
  ]
  rendererClasses [&NiagaraSpriteRendererProperties]
  renderers [{rendererClass=&NiagaraSpriteRendererProperties}]
 }
}
summary {
 systemName=NS_Shoe_TabletSoilTray_Trail_seq
 emitters [
  {bEnabled=false, emitterName=Smoke, rendererClasses [&NiagaraSpriteRendererProperties]}
  {emitterName=FN_AttachedRibbon001, rendererClasses [&NiagaraRibbonRendererProperties]}
 ]
 userVariables [
  User.bIsFrontEnd {type=NiagaraBool}
  User.bIsFrontEndPreview {type=NiagaraBool}
  User.HideFootFX {type=NiagaraBool}
  User.isOffPawnPreview {type=NiagaraBool}
  User.IsSkydiving {type=NiagaraBool}
  User.LandBurst {type=NiagaraFloat, defaultValue=0, description=""}
  User.RibbonIDOffset {type=NiagaraInt32}
  User.SkeletalMesh {
   type=NiagaraDataInterfaceSkeletalMesh
   defaultValue {
    struct=NiagaraExt_VariableValue_DataInterface
    value {dataInterface=&self:NiagaraDataInterfaceSkeletalMesh_0}
   }
  }
  User.TrailOn {type=NiagaraFloat, defaultValue=1, description=""}
 ]
}
