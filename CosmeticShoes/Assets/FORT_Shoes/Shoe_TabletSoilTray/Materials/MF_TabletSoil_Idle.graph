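# Material function for the TabletSoilTray cosmetic shoe that adds an animated idle glow to incoming material attributes, with landing and run controls.
# An animated Fresnel edge glow cycles four colors through panning noise and sine timing, then is combined by Max with a two-color bottom glow.
# Sphere masks and noise driven by pre-skinned position shape bottom bands and landing speed lines.
# TurnOffEmissive scales the glow, which feeds both Emissive and a darkened BaseColor.
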MaterialFunction /CosmeticShoes/Assets/FORT_Shoes/Shoe_TabletSoilTray/Materials/MF_TabletSoil_Idle
graph {nodes [
  Constant()
  Constant()
  n1 = GetMaterialAttributes(None=FunctionInput() {inputType=FunctionInput_MaterialAttributes}) {attributeGetTypes [#1]}
  n2 = NamedRerouteUsage()
  n3 = Constant()
  n4 = ScalarParameter() {defaultValue=15.5, group=TabletSoil, parameterName=LandingRadiusFresnel}
  n5 = Constant() {r=-0.8878}
  n6 = Multiply(
   A=NamedRerouteUsage()
   B=LinearInterpolate(B=Constant() {r=-5.5}, Alpha=SphereMask(A=@n2, B=@n3, Radius=@n4, Hardness=@n5)) {constA=0.8, constB=-0.5}
  ) {constB=0.7}
  n7 = Step(X=ComponentMask(None=TextureCoordinate()) {g=true}) {constX=0.1, constY=0.1}
  n8 = NamedRerouteUsage()
  n9 = Saturate(None=Multiply(A=Subtract(A=@n8) {constB=0.5}) {constB=3})
  n10 = Multiply(
   A=NamedRerouteUsage()
   B=Multiply(A=ScalarParameter() {defaultValue=0.5, group=TabletSoil, parameterName=Global edge light intensity}, B=Multiply(A=@n7, B=@n9))
  )
  n11 = LinearInterpolate(Alpha=ScalarParameter() {group=TabletSoil, parameterName=IsRun}) {constA=1, constB=2}
  n12 = VectorParameter() {group=TabletSoil, parameterName=Clolor_Bottom_A, defaultValue=<rgba 0.8503 0.2376 2 1>}
  n13 = VectorParameter() {group=TabletSoil, parameterName=Clolor_Bottom_B, defaultValue=<rgba 2 0.1 1.8511 1>}
  n14 = TextureSample(UVs=NamedRerouteUsage()) {samplerType=SAMPLERTYPE_Grayscale, texture=&LowResBlurredNoise}
  n15 = Add(A=Multiply(A=NamedRerouteUsage(), B=ScalarParameter() {group=TabletSoil, parameterName=landingBlend}), B=NamedRerouteUsage())
  n16 = ScalarParameter() {defaultValue=0.5, group=TabletSoil, parameterName=Global Bottom light intensity}
  n17 = Max(
   A=Multiply(A=OneMinus(None=@n6), B=Multiply(A=@n10, B=@n11))
   B=Multiply(A=LinearInterpolate(A=@n12.RGB, B=@n13.RGB, Alpha=@n14.R), B=Multiply(A=@n15, B=@n16))
  )
  n18 = LinearInterpolate(B=ScalarParameter() {defaultValue=0.3, group=TabletSoil, parameterName=ShoeSole_Brightness_Mul}, Alpha=OneMinus(None=NamedRerouteUsage())) {constA=0.05, constB=0.65}
  n19 = Multiply(
   A=LinearInterpolate(A=@n17, B=Multiply(A=@n17, B=@n18), Alpha=NamedRerouteUsage())
   B=ScalarParameter() {defaultValue=1, group=TabletSoil, parameterName=TurnOffEmissive}
  )
  n20 = MaterialFunctionCall(Emissive=@n19) {materialFunction=&MF_FX_EmissiveAdjustments}
  n21 = Multiply(A=NamedRerouteUsage(), B=Constant() {r=30}) {constB=1000}
  n22 = LinearInterpolate(
   A=@n1.BaseColor
   B=Constant3Vector() {constant=<rgba 0 0 0 1>}
   Alpha=Multiply(A=Saturate(None=@n21), B=ScalarParameter() {defaultValue=1, group=TabletSoil, parameterName=TurnOffEmissive})
  ) {constB=0.2}
  FunctionOutput(None=SetMaterialAttributes!(MaterialAttributes=@n1.MaterialAttributes, EmissiveColor=Multiply(A=@n20.Result) {constB=0.05}, BaseColor=Add(A=@n22, B=@n19)) {attributeSetTypes [#2, #1]})
  n23 = NamedRerouteUsage()
  n24 = Constant2Vector() {r=0.8}
  LinearInterpolate(
   A=TextureSample(UVs=Panner(Coordinate=@n23, Speed=@n24) {speedX=-1}) {samplerType=SAMPLERTYPE_Grayscale, texture=&LowResBlurredNoise}.G
   Alpha=NamedRerouteUsage()
  )
  n25 = TextureSample() {samplerType=SAMPLERTYPE_Masks, texture=/CosmeticShoes/Assets/FORT_Shoes/Shoe_TabletSoilTray/Textures/T_TabletSoilTray_Shoe_mask}
  NamedRerouteDeclaration(None=@n25.B) {name=Inside, nodeColor=<rgba 0 1 0.0156 1>}
  NamedRerouteDeclaration(None=@n25.G) {name=ShoeSole, nodeColor=<rgba 0.7344 1 0 1>}
  NamedRerouteDeclaration(None=ScalarParameter() {group=TabletSoil, parameterName=landingBlend}) {name=Landing, nodeColor=<rgba 0 1 0.1094 1>}
  NamedRerouteDeclaration(None=ScalarParameter() {group=TabletSoil, parameterName=IsRun}) {name=Run, nodeColor=<rgba 1 0 0.2109 1>}
  NamedRerouteDeclaration(None=ScalarParameter() {defaultValue=1, group=TabletSoil, parameterName=TurnOffEmissive}) {name=TurnOffEmissive, nodeColor=<rgba 0 0.5312 1 1>}
  n26 = LinearInterpolate(
   A=ScalarParameter() {defaultValue=100, group=TabletSoil, parameterName=landingBrightnessOffset}
   B=ScalarParameter() {group=TabletSoil, parameterName=ShoeSloe_Brightness}
   Alpha=NamedRerouteUsage()
  )
  n27 = SphereMask(
   A=NamedRerouteUsage()
   B=Constant()
   Radius=ScalarParameter() {defaultValue=1.4, group=TabletSoil, parameterName=LandingRadius}
   Hardness=ScalarParameter() {defaultValue=-0.3712, group=TabletSoil, parameterName=LandingRadius_Hardness}
  )
  NamedRerouteDeclaration(None=LinearInterpolate(B=LinearInterpolate(B=@n26, Alpha=@n27) {constA=1}, Alpha=NamedRerouteUsage()) {constA=1}) {name=landingBrightness, nodeColor=<rgba 1 0.9844 0 1>}
  n28 = NamedRerouteDeclaration(None=FunctionInput(Preview=Constant3Vector() {constant=<rgba 0 0 0 1>}) {inputName=preskinPosition, sortPriority=1}) {name=PreskinPosition, nodeColor=<rgba 0.8047 1 0 1>}
  NamedRerouteDeclaration(None=VertexInterpolator(VS=ComponentMask(None=@n28) {b=true}).PS) {name=PreskinPosition_B, nodeColor=<rgba 1 0.2109 0 1>}
  NamedRerouteDeclaration(None=@n25.R) {name=LeftRightMask, nodeColor=<rgba 0.6641 0 1 1>}
  n29 = TextureSampleParameter2D() {
   group=00_Texture Inputs
   parameterName=Mask
   samplerType=SAMPLERTYPE_Masks
   sortPriority=3
   texture=&T_BlackMask
  }
  NamedRerouteDeclaration(None=@n29.R) {name=AO, nodeColor=<rgba 0 0.4609 1 1>}
  n30 = NamedRerouteUsage()
  n31 = LinearInterpolate(Alpha=TextureSample(UVs=Multiply(A=@n30) {constB=0.8}) {samplerType=SAMPLERTYPE_Grayscale, texture=&LowResBlurredNoise}.R) {constA=1, constB=0.3}
  n32 = Time()
  n33 = Sine(None=Multiply(A=Add(A=@n32) {constB=0.5}) {constB=0.5})
  n34 = Sine(None=Multiply(A=Time()) {constB=0.3})
  n35 = NamedRerouteUsage()
  n36 = Panner(Coordinate=Divide(A=ComponentMask(None=@n35) {b=true, g=true}, B=ScalarParameter() {defaultValue=42, group=TabletSoil, parameterName=Color_Noise_tiling}) {constB=1024}) {speedX=0.05, speedY=0.15}
  n37 = Multiply(A=TextureSample(UVs=NamedRerouteDeclaration(None=@n36) {name=panner_Animation_Uv, nodeColor=<rgba 0.875 0 1 1>}) {samplerType=SAMPLERTYPE_Grayscale, texture=&LowResBlurredNoise}.R) {constB=2}
  n38 = LinearInterpolate(
   A=LinearInterpolate(
    A=VectorParameter() {group=TabletSoil, parameterName=Clolor_Edge_D, defaultValue=<rgba 0.1112 0.104 1.3 1>}.RGB
    B=VectorParameter() {group=TabletSoil, parameterName=Clolor_Edge_A, defaultValue=<rgba 0.0994 0.0962 0.85 1>}.RGB
    Alpha=Saturate(None=@n33)
   )
   B=LinearInterpolate(
    A=VectorParameter() {group=TabletSoil, parameterName=Clolor_Edge_C, defaultValue=<rgba 1.2888 0.112 1.4 1>}.RGB
    B=VectorParameter() {group=TabletSoil, parameterName=Clolor_Edge_B, defaultValue=<rgba 0.65 0.1058 0.85 1>}.RGB
    Alpha=Saturate(None=@n34)
   )
   Alpha=Saturate(None=Power(Base=@n37))
  )
  n39 = ScalarParameter() {defaultValue=0.75, group=TabletSoil, parameterName=LandingFresnelExponentln}
  n40 = NamedRerouteUsage()
  n41 = Multiply(A=Time()) {constB=0.3}
  n42 = LinearInterpolate(
   A=ScalarParameter() {defaultValue=2.3, group=TabletSoil, parameterName=TabletSoi_Fresnel_Exponentln_MIn_A}
   B=ScalarParameter() {defaultValue=2.2, group=TabletSoil, parameterName=TabletSoi_Fresnel_Exponentln_MIn_B}
   Alpha=Saturate(None=Sine(None=@n41))
  )
  n43 = ScalarParameter() {defaultValue=2.4, group=TabletSoil, parameterName=TabletSoi_Fresnel_Exponentln_Max}
  n44 = ComponentMask(None=NamedRerouteUsage()) {b=true, g=true}
  n45 = ScalarParameter() {defaultValue=14.1051, group=TabletSoil, parameterName=Fresnel_Noise_tiling}
  n46 = TextureSample(UVs=Panner(Coordinate=Divide(A=@n44, B=@n45) {constB=1024}) {speedX=0.2, speedY=-0.2}) {samplerType=SAMPLERTYPE_Grayscale, texture=&LowResBlurredNoise}
  n47 = Saturate(None=Power(Base=Multiply(A=@n46.R) {constB=2}))
  n48 = Multiply(
   A=LinearInterpolate(Alpha=ScalarParameter() {group=TabletSoil, parameterName=IsRun}) {constA=1, constB=0.8}
   B=Multiply(A=LinearInterpolate(B=@n39, Alpha=@n40) {constA=1, constB=0.75}, B=LinearInterpolate(A=@n42, B=@n43, Alpha=@n47))
  )
  n49 = Power(
   Base=Multiply(A=Fresnel(ExponentIn=@n48), B=ScalarParameter() {defaultValue=2.3, group=TabletSoil, parameterName=TabletSoi_Fresnel_Mask_mul})
   Exp=ScalarParameter() {defaultValue=2.8, group=TabletSoil, parameterName=TabletSoi_Fresnel_power}
  )
  NamedRerouteDeclaration(None=Multiply(A=Multiply(A=@n31, B=@n38), B=Saturate(None=@n49))) {name=The edges of the shoes glow, nodeColor=<rgba 0 0.5312 1 1>}
  n50 = NamedRerouteUsage()
  n51 = Constant() {r=-0.05}
  n52 = ScalarParameter() {defaultValue=6.4128, group=TabletSoil, parameterName=SoftEdge}
  n53 = Constant() {r=-6.6723}
  n54 = Multiply(A=Power(Base=SphereMask(A=@n50, B=@n51, Radius=@n52, Hardness=@n53))) {constB=0.5}
  n55 = SphereMask(
   A=NamedRerouteUsage()
   B=Constant() {r=-0.05}
   Radius=LinearInterpolate(
    A=ScalarParameter() {defaultValue=0.71, group=TabletSoil, parameterName=HardEdge}
    B=ScalarParameter() {defaultValue=1.237, group=TabletSoil, parameterName=landingEdge}
    Alpha=ScalarParameter() {group=TabletSoil, parameterName=landingBlend}
   ) {constB=1.5}
   Hardness=Constant() {r=-1.9428}
  )
  n56 = Constant() {r=30}
  n57 = TextureSample(UVs=Multiply(A=NamedRerouteUsage()) {constB=1.2}) {samplerType=SAMPLERTYPE_Grayscale, texture=&LowResBlurredNoise}
  n58 = Multiply(
   A=SphereMask(A=NamedRerouteUsage(), B=Constant(), Radius=Constant() {r=52.328}, Hardness=Constant() {r=-0.5483})
   B=Multiply(A=LinearInterpolate(A=@n56, Alpha=@n57.R) {constB=15}, B=NamedRerouteUsage())
  )
  n59 = Constant() {r=0.8925}
  n60 = Power(Base=OneMinus(None=Fresnel(ExponentIn=@n59)), Exp=Constant() {r=1.2})
  NamedRerouteDeclaration(None=Multiply(A=Max(A=@n54, B=@n55), B=Multiply(A=@n58, B=@n60)) {constB=100}) {name=Idle, nodeColor=<rgba 0.4297 1 0 1>}
  n61 = NamedRerouteUsage()
  NamedRerouteDeclaration(None=Divide(A=ComponentMask(None=@n61) {g=true, r=true}, B=ScalarParameter() {defaultValue=3.5, group=TabletSoil, parameterName=PreSkinPosition_UV_Tile}) {constB=1024}) {name=PreSkinPosition_UV, nodeColor=<rgba 0.4531 1 0 1>}
  n62 = Constant() {r=3}
  n63 = Panner(Coordinate=NamedRerouteUsage(), Speed=Constant2Vector() {r=-0.8}) {speedX=-1}
  n64 = LinearInterpolate(
   A=Multiply(A=@n62) {constB=-1}
   B=@n62
   Alpha=MaterialFunctionCall(
    In=TextureSample(UVs=@n63) {samplerType=SAMPLERTYPE_Grayscale, texture=&LowResBlurredNoise}.G
    Contrast=Constant() {r=0.2785}
   ) {materialFunction=&CheapContrast}.Result
  ) {constA=-2, constB=2}
  n65 = Constant() {r=-1.2266}
  NamedRerouteDeclaration(None=Power(Base=Multiply(A=@n64, B=@n65), Exp=Constant() {r=1})) {name=PositionNoise, nodeColor=<rgba 0 0.6953 1 1>}
  n66 = NamedRerouteUsage()
  n67 = NamedRerouteUsage()
  n68 = NamedRerouteUsage()
  n69 = Constant2Vector() {g=0.02}
  n70 = MaterialFunctionCall(
   In=TextureSample(UVs=Panner(Coordinate=@n68, Speed=@n69) {speedX=-1}) {samplerType=SAMPLERTYPE_Grayscale, texture=&LowResBlurredNoise}.R
   Contrast=Constant() {r=0.6}
  ) {materialFunction=&CheapContrast}
  n71 = LinearInterpolate(
   A=ScalarParameter() {defaultValue=5.2, group=TabletSoil, parameterName=LandingSpeedLine_Radius_MIn}
   B=ScalarParameter() {defaultValue=7.5, group=TabletSoil, parameterName=LandingSpeedLine_Radius_Max}
   Alpha=@n70.Result
  )
  n72 = ScalarParameter() {defaultValue=1, group=TabletSoil, parameterName=FrontEnd_SpeedLine_Radius}
  n73 = Power(
   Base=SphereMask(
    A=Add(A=@n66, B=@n67)
    B=Constant() {r=0.5073}
    Radius=Multiply(A=@n71, B=@n72)
    Hardness=ScalarParameter() {defaultValue=-4.3707, group=TabletSoil, parameterName=LandingSpeedLine_Hardness}
   )
   Exp=Constant() {r=1.8}
  )
  n74 = LinearInterpolate(
   A=ScalarParameter() {defaultValue=30, group=TabletSoil, parameterName=LandingSpeedLine_Brightness_Min}
   B=ScalarParameter() {defaultValue=100, group=TabletSoil, parameterName=LandingSpeedLine_Brightness_Max}
   Alpha=@n70.Result
  )
  NamedRerouteDeclaration(None=LinearInterpolate(A=Multiply(A=@n73, B=@n74), Alpha=NamedRerouteUsage()) {constB=0}) {name=LandingSpeedLine, nodeColor=<rgba 1 0.9844 0 1>}
  NamedRerouteUsage()
 ]}
