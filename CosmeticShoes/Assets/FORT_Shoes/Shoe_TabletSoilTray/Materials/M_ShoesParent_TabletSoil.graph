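# Parent surface material for the TabletSoilTray cosmetic shoe, using material attributes and a shading model set by the expression graph.
# The 2023 character PreFX base feeds a borrowed backpack fluffy-fabric function, masked by Tech Art Mask times a UV U step.
# TabletSoil idle (driven by pre-skinned position), shoe reactivity test and sock functions follow.
# Character PostFX then feeds the final shoe hologram function.
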Material /CosmeticShoes/Assets/FORT_Shoes/Shoe_TabletSoilTray/Materials/M_ShoesParent_TabletSoil
graph {
 properties {
  bUseMaterialAttributes=true
  bUsedWithClothing=true
  bUsedWithMeshDeformer=true
  bUsedWithMorphTargets=true
  bUsedWithNanite=true
  bUsedWithNiagaraMeshParticles=true
  bUsedWithSkeletalMesh=true
  pixelDepthOffsetMode=PDOM_Legacy
  shadingModel=MSM_FromMaterialExpression
  usedShadingModels=Default Lit | Subsurface | Clear Coat | Subsurface Profile | Cloth
 }
 nodes [
  n1 = MaterialFunctionCall() {materialFunction=&MF_CharacterParent_2023_PreFX}
  n2 = TextureCoordinate()
  n3 = Multiply(A=@n1.Tech Art Mask, B=Step(Y=ComponentMask(None=@n2) {r=true}))
  n4 = MaterialFunctionCall!(Material=MaterialFunctionCall!(
    In=MaterialFunctionCall!(Material=@n1.Result, M=@n1.Result, TechArtMask=@n3) {materialFunction=/BRCosmetics/Accessories/FORT_Backpacks/Backpack_M_MED_PearlTote/Materials/MF_FluffyFabric}.Result
    preskinPosition=@n1.Pre skinned Local Position
   ) {materialFunction=&/MF_TabletSoil_Idle}.Result) {materialFunction=/Game/Global/Materials/Cosmetics/Kicks/Shoes/Materials/MF_ShoeReactivityTest}
  MaterialFunctionCall!(In=MaterialFunctionCall(
    Material=MaterialFunctionCall!(InOpacity, Material=@n4.Result, Mask=@n1.Tech Art Mask) {materialFunction=/Game/Global/Materials/Cosmetics/Kicks/Socks/Materials/MF_Sock}.Result
    M=@n1.M
    Pre Skinned Local Position=@n1.M
    Distance Blend=@n1.M
    Fresnel=@n1.M
   ) {materialFunction=&MF_CharacterParent_2023_PostFX}.Result) {materialFunction=/Game/Global/Materials/Cosmetics/Kicks/Shoes/Materials/MF_TechartShoeHologram}
 ]
}
